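# Sound cue for the Sci Ops glider cosmetic's thrust sound, drawing on four thrust-start waves.
# A Random node picks one of four equally weighted wave players on each play.
# Playback uses the Athena player glider thrust attenuation and concurrency, the gliders land sound class and a cosmetic glider loop audio properties sheet, at reduced volume.
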SoundCue /BRCosmetics/Sounds/Gliders/Sci_Ops/Glider_Sci_Ops_Thrust_Cue
root=Random_0
nodes [
 {id=Random_0, children [WavePlayer_1,WavePlayer_2,WavePlayer_3,WavePlayer_4], properties {weights [1,1,1,1]}}
 {id=WavePlayer_1, properties {soundWaveAssetPtr=&/Glider_Sci_Ops_Thrust_Start_01}}
 {id=WavePlayer_2, properties {soundWaveAssetPtr=&/Glider_Sci_Ops_Thrust_Start_02}}
 {id=WavePlayer_3, properties {soundWaveAssetPtr=&/Glider_Sci_Ops_Thrust_Start_03}}
 {id=WavePlayer_4, properties {soundWaveAssetPtr=&/Glider_Sci_Ops_Thrust_Start_04}}
]
properties {
 attenuationSettings=&Fort_Player_Glider_Thrust_Athena_Att
 audioPropertiesSheet=&APS_Cosmetic_Glider_Loop
 maxDistance=12500
 soundClassObject=&Fort_Gliders_Land
 volumeMultiplier=0.7
 concurrencySet [&Athena_Glider_Thrust_Conc]
}
